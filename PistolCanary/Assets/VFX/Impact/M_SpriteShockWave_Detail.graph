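# Translucent unlit Niagara sprite material for an impact shockwave, using 2D-offset refraction.
# Refraction direction radiates from UV center with optional axis swizzle, scaled by masks, temporal dither and the particle Ref dynamic parameter.
# Masks include a sphere mask, a power-shaped ring ramp and polar-mapped smoke noise.
# Opacity combines particle and tint alpha with depth and camera fades, while particle color feeds the shared emissive-adjustment function.
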Material /PistolCanary/Assets/VFX/Impact/M_SpriteShockWave_Detail
graph {
 properties {
  bAutomaticallySetUsageInEditor=false
  bUsedWithNiagaraSprites=true
  blendMode=BLEND_Translucent
  refractionMethod=RM_2DOffset
  shadingModel=MSM_Unlit
  translucencyPass=MTP_BeforeDOF
  usedShadingModels=Unlit
 }
 nodes [
  LinearInterpolate(A=TextureCoordinate(), B=TextureCoordinate() {coordinateIndex=1}, Alpha=ScalarParameter() {parameterName=CubeMap_Switch})
  MaterialFunctionCall(Emissive=ParticleColor().RGB) {materialFunction=&MF_FX_EmissiveAdjustments}
  n1 = Subtract(A=TextureCoordinate()) {constB=0.5}
  n2 = Normalize(VectorInput=LinearInterpolate(
    A=@n1
    B=MaterialFunctionCall(XY=@n1) {materialFunction=&Swizzle}.YX
    Alpha=ScalarParameter() {parameterName=SwizzleRefDir}
   ))
  n3 = Constant2Vector() {g=-1, r=1}
  n4 = Multiply(
   A=NamedRerouteUsage()
   B=Multiply(
    A=MaterialFunctionCall() {materialFunction=&DitherTemporalAA}.Result
    B=ScalarParameter() {defaultValue=0.2, parameterName=DTAA}
   )
  )
  n5 = NamedRerouteUsage()
  Multiply(
   A=Multiply(A=Multiply(A=@n2, B=@n3), B=Add(A=@n4, B=@n5))
   B=DynamicParameter() {defaultValue=<rgba 1 1 1 0.1>, paramNames [Time,Distort,Dissolve,Ref]}.Ref
  )
  n6 = Constant() {r=0.5}
  n7 = Constant() {r=0.5}
  Multiply(
   A=NamedRerouteUsage()
   B=TextureSample(UVs=MaterialFunctionCall!(UV Coord, LenghtScale=@n6, RadialScale=@n7) {materialFunction=/Game/Global/MaterialFunctions/TechArt/MF_PolarCoordinates}.Result) {texture=&T_Inky_Smoke_Tile}.R
  )
  n8 = OneMinus(None=NamedRerouteUsage())
  n9 = ScalarParameter() {defaultValue=5, parameterName=AlphaRampPower}
  n10 = Subtract(A=Multiply(A=Power(Base=@n8, Exp=@n9) {constExponent=4}) {constB=2})
  n11 = MaterialFunctionCall(Float2=OneMinus(None=Abs(None=@n10))) {materialFunction=&BreakOutFloat2Components}
  NamedRerouteDeclaration(None=MaterialFunctionCall(A=@n11.R, B=Multiply(A=@n11.R, B=@n11.R), C=@n11.R, Alpha=ScalarParameter() {defaultValue=1, parameterName=AlphaMap_Select}) {materialFunction=&Lerp_3Color}.Result) {name=CubeMap, nodeColor=<rgba 1 1 1 1>}
  n12 = ScalarParameter() {defaultValue=20, parameterName=Camfade_Offset}
  n13 = ScalarParameter() {defaultValue=300, parameterName=Camfade_Lemgth}
  n14 = Divide(A=Subtract(A=PixelDepth().R, B=@n12), B=@n13)
  NamedRerouteDeclaration(None=Saturate(None=@n14)) {name=CamFade, nodeColor=<rgba 1 0 0.9613 1>}
  n15 = Constant() {r=0.5}
  NamedRerouteDeclaration(None=SphereMask(A=TextureCoordinate(), B=@n15, Radius=@n15) {hardnessPercent=0}) {name=SimpleSphereMask, nodeColor=<rgba 1 1 1 1>}
  n16 = MaterialFunctionCall!(
   InAlpha=NamedRerouteUsage()
   OpacityBasedFade_DistanceA=ScalarParameter() {defaultValue=200, group=Opacity, parameterName=OpacityBasedFade_DistanceA}
   OpacityBasedFade_DistanceB=ScalarParameter() {defaultValue=50, group=Opacity, parameterName=OpacityBasedFade_DistanceB}
  ) {materialFunction=/Game/Effects/Niagara/MasterMaterials/Functions/MF_FX_OpacityBasedDepthFade}
  NamedRerouteDeclaration(None=@n16.FadeValue) {name=DepthFadeValue, nodeColor=<rgba 0 1 0.5312 1>}
  n17 = Multiply(A=VectorParameter() {parameterName=EmissiveOverAllTint, defaultValue=<rgba 1 1 1 1>}.A, B=ParticleColor().A)
  NamedRerouteDeclaration(None=Multiply(A=NamedRerouteUsage(), B=Multiply(A=@n17, B=@n16.FinalAlpha))) {name=Opacity, nodeColor=<rgba 0 0 0 1>}
  n18 = Subtract(A=CameraPositionWS(), B=WorldPosition().XYZ)
  n19 = Constant3Vector() {constant=<rgba 0 0 -1 0>}
  n20 = Multiply(
   A=Length(None=@n18) {is=class MaterialExpressionLength}
   B=DotProduct(A=Normalize(VectorInput=@n18), B=Transform(None=@n19) {transformSourceType=TRANSFORMSOURCE_View})
  )
  StaticSwitch(True=Divide(A=Subtract(A=@n20, B=@n12), B=@n13), False=@n14, Value=StaticBool())
  TextureCoordinate()
 ]
}
